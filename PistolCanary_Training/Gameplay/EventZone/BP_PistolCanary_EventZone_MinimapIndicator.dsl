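; Minimap indicator actor for the PistolCanary training event zone, which also switches off IcicleLarva spawning.
; On BeginPlay it waits asynchronously for playlist data to be ready, then broadcasts a stateful IcicleLarva event with false through the global event router and logs the disable.
; It then destroys its Skydrop and Player AthenaSpecialActor components and activates the Event one, leaving that as the remaining indicator.
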(event EventBeginPlay
  (Class|FortAsyncActionCurrentPlaylistReady|PlaylistDataReadyAsync
    (:OnReady
      (GameplayEventRouter|BroadcastStatefulEvent (GameplayEventRouter|GetGlobalEventRouter) (Utilities|Struct|MakeIcicleLarvaEventStruct false) 0)
      (Fortnite|FortShippingLog 0 "[BP_PistolCanary_EventZone_MinimapIndicator] - Disable IcicleLarva spawn.")
      (Components|DestroyComponent (Variables|Default|GetAthenaSpecialActor_Skydrop))
      (Components|DestroyComponent (Variables|Default|GetAthenaSpecialActor_Player))
      (Components|Activation|SetActive (Variables|Default|GetAthenaSpecialActor_Event) true false))))

(fn ConstructionScript)
