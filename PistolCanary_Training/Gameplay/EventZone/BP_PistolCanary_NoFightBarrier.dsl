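; No-fight barrier zone for the PistolCanary event, granting human players damage-rule, no-ammo-cost and aura effects plus leaderboard entry.
; Server overlaps apply and clear effects, disable vehicles and damage non-exempt bots pre-event; clients manage leaderboard input.
; Calendar or router events start the event, clearing PvP indicators and pre-event effects.
; Event end shrinks and moves the barrier via timeline, or hides it under a hotfixable flag; OnReps sync visibility and scale.
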(event Custom|RefreshPlayersInCombat
  (params (PlayerPawn :type Fort%20Player%20Pawn%20Object%20Reference))
  (bind _returnvalue (Fortnite|GetFortPlayerControllerfromActor PlayerPawn))
  (bind _outfortplayerpawns (Fortnite|GetFortPlayerPawns))
  (for _array_element _outfortplayerpawns
    (if (Utilities|Array|ContainsItem (Variables|Default|GetCombatedActors) _array_element)
      (IndicatedManagement|AddActorsToIndicatedList _returnvalue (Variables|Default|GetCombatedActors) (IndicatedManagement|MakeIndicatedActorData :ScaleAndOpacityData (IndicatedManagement|MakeIndicatedActorScaleandOpacityData 7.5e+03 1e+01 0.5 2.0) :Duration -1.0 :StepTime 0.1 :IndicatorOffset "0.000000,0.000000,90.000000" :IndicatorDBNOOffset "0.000000,0.000000,30.000000" :ShareActorWith "None" :StateImageOverride "None" :StateIndexOverride -1 :bClampToScreen false :bHideStateImage false :bOnlyDisplayWhenStencilingIsDisabled false :bOnlyDisplayIfActorIsVisible false) false false true)
      (else
        (IndicatedManagement|RemoveActorFromIndicatedList _returnvalue _array_element)))))

(event Custom|PlayersAddedToPvP
  (params (PlayerPawn :type Actor%20Object%20Reference))
  (Utilities|Array|AddUnique (Variables|Default|GetCombatedActors) PlayerPawn)
  (for _array_element (Collision|GetOverlappingActors (Variables|Default|GetOverlap_Volume) "/Script/FortniteGame.FortPlayerPawn" :!class /Script/Engine.PrimitiveComponent)
    (Ability|SendGameplayEventtoActor _array_element "(TagName=\"PistolCanary.PreEvent.PvP.Indicator.Added\")" (Utilities|Struct|MakeGameplayEventData "(TagName=\"\")" PlayerPawn))))

(event Custom|RemovePlayerFromPvP
  (params (PlayerPawn :type Actor%20Object%20Reference%20%28by%20ref%29))
  (bind _combatedactors (Variables|Default|GetCombatedActors))
  (for _array_element _combatedactors
    (IndicatedManagement|RemoveActorFromIndicatedList (Fortnite|GetFortPlayerControllerfromActor _array_element) PlayerPawn)
    (Development|PrintString "Remove Self from indicator list"))
  (Utilities|Array|RemoveItem _combatedactors PlayerPawn)
  (for _array_element_2 (Collision|GetOverlappingActors (Variables|Default|GetOverlap_Volume) "/Script/FortniteGame.FortPlayerPawn" :!class /Script/Engine.PrimitiveComponent)
    (Ability|SendGameplayEventtoActor _array_element_2 (select (Variables|Default|GetEventStarted) "(TagName=\"PistolCanary.PreEvent.PvP.Indicator.EventClear\")" "(TagName=\"PistolCanary.PreEvent.PvP.Indicator.Removed\")") (Utilities|Struct|MakeGameplayEventData "(TagName=\"\")" PlayerPawn))
    (Development|PrintString "Remove external actors from indicator list.")))

(event Custom|PC_ShowSafeZone
  ;[0 0.15 0.05285 0.5] Cheats to show/hide the safe zone
  (Variables|Default|SetShowZone true))

(event Custom|PC_HideSafeZone
  ;[0 0.15 0.05285 0.5] Cheats to show/hide the safe zone
  (Variables|Default|SetShowZone false))

(event Custom|EventHasStarted
  (bind _playersingame (Variables|Default|GetPlayersInGame))
  (Variables|Default|SetEventStarted true)
  (bind _outfortplayerpawns (Fortnite|GetFortPlayerPawns))
  (Variables|Default|SetPlayersInGame _outfortplayerpawns)
  (for _array_element _playersingame
    (CallFunction|RemovePlayerfromPvP _array_element) ; Clear players from PVP first to clear indicators correctly.
  )
  (for _array_element_2 _playersingame
    (Utilities|IsValid (Ability|GetFortAbilitySystemComponent _array_element_2)
      (:Is%20Valid
        (GameplayEffects|RemoveActiveGameplayEffectbySourceEffect (Ability|GetFortAbilitySystemComponent _array_element_2) (Variables|Default|GetKillCountPassive) (Ability|GetFortAbilitySystemComponent _array_element_2) -1)
        (GameplayEffects|RemoveActiveGameplayEffectbySourceEffect (Ability|GetFortAbilitySystemComponent _array_element_2) (Variables|Default|GetGEDamageImmune) (Ability|GetFortAbilitySystemComponent _array_element_2) -1)
        (GameplayEffects|RemoveActiveGameplayEffectbySourceEffect (Ability|GetFortAbilitySystemComponent _array_element_2) (Variables|Default|GetGEDamageRules) (Ability|GetFortAbilitySystemComponent _array_element_2) -1)
        (Variables|Default|SetRemoveLeaderBoard (+ (Variables|Default|GetRemoveLeaderBoard) 1)))))
  (bind _returnvalue (Actor|GetActorOfClass "/PistolCanary_Training/Gameplay/BP_PistolCanary_EventZone_PreEvent_GameplayManager.BP_PistolCanary_EventZone_PreEvent_GameplayManager_C"))
  (Class|BPPistolCanaryEventZonePreEventGameplayManager|SetEventStarted _returnvalue :!class /PistolCanary_Training/Gameplay/BP_PistolCanary_EventZone_PreEvent_GameplayManager.BP_PistolCanary_EventZone_PreEvent_GameplayManager_C))

(event EventBeginPlay
  (bind _playersingame (Variables|Default|GetPlayersInGame))
  (|CallParentFunction)
  (if (Variables|Default|GetbDebugActivateZone)
    (CallFunction|PCShowSafeZone)
    (else
      (Networking|SwitchHasAuthority
        (:Authority
          (CallFunction|CheckLoadedwithCheats)
          (bind (_asynctaskproxy _payload) (GameplayEventRouter|StartListeningToStatefulEvent (GameplayEventRouter|GetGlobalEventRouter) "/Events_Shared/Gameplay/GlobalEvents/Event_RouterEvent_String.Event_RouterEvent_String")
            (:then
              (if (Editor|InEditororPIE)
                (CallFunction|PCShowSafeZone)
                (elif (Variables|Default|GetFollowCalendarEvents)
                  (Class|FortAsyncActionAuthorityMatchReady|MatchReadyAsync
                    (:OnReady
                      (MCPContext|InitialCalendarSyncCompleted
                        (:OnSyncComplete
                          (bind _biseventactive (CalendarEvents|GetCalendarEventInformation (Variables|Default|GetCalendarEventToWatch)))
                          (if _biseventactive
                            (Rendering|SetActorHiddenInGame false)
                            (MCPContext|WatchCalendarEvent (Variables|Default|GetMainEventCalendar)
                              (:OnEventActive
                                (Variables|Default|SetEventStarted true)
                                (bind _outfortplayerpawns (Fortnite|GetFortPlayerPawns))
                                (Variables|Default|SetPlayersInGame _outfortplayerpawns)
                                (for _array_element _playersingame
                                  (CallFunction|RemovePlayerfromPvP _array_element) ; Clear players from PVP first to clear indicators correctly.
                                )
                                (for _array_element_2 _playersingame
                                  (Utilities|IsValid (Ability|GetFortAbilitySystemComponent _array_element_2)
                                    (:Is%20Valid
                                      (GameplayEffects|RemoveActiveGameplayEffectbySourceEffect (Ability|GetFortAbilitySystemComponent _array_element_2) (Variables|Default|GetKillCountPassive) (Ability|GetFortAbilitySystemComponent _array_element_2) -1)
                                      (GameplayEffects|RemoveActiveGameplayEffectbySourceEffect (Ability|GetFortAbilitySystemComponent _array_element_2) (Variables|Default|GetGEDamageImmune) (Ability|GetFortAbilitySystemComponent _array_element_2) -1)
                                      (GameplayEffects|RemoveActiveGameplayEffectbySourceEffect (Ability|GetFortAbilitySystemComponent _array_element_2) (Variables|Default|GetGEDamageRules) (Ability|GetFortAbilitySystemComponent _array_element_2) -1)
                                      (Variables|Default|SetRemoveLeaderBoard (+ (Variables|Default|GetRemoveLeaderBoard) 1)))))
                                (bind _returnvalue (Actor|GetActorOfClass "/PistolCanary_Training/Gameplay/BP_PistolCanary_EventZone_PreEvent_GameplayManager.BP_PistolCanary_EventZone_PreEvent_GameplayManager_C"))
                                (Class|BPPistolCanaryEventZonePreEventGameplayManager|SetEventStarted _returnvalue :!class /PistolCanary_Training/Gameplay/BP_PistolCanary_EventZone_PreEvent_GameplayManager.BP_PistolCanary_EventZone_PreEvent_GameplayManager_C)))))))))))
            (:OnEventReceived
              (if (Utilities|String|Equal,CaseInsensitive(String) (Utilities|Struct|BreakEventRouterEventString _payload) (Variables|Default|GetInventoryRouterEvent))
                (Variables|Default|SetCheatActivated false)
                (elif (Utilities|String|Equal,CaseInsensitive(String) (Utilities|Struct|BreakEventRouterEventString _payload) (Variables|Default|GetEventStartRouterEvent))
                  (Variables|Default|SetEventStarted true)
                  (CallFunction|EventHasStarted)
                  (elif (Utilities|String|Equal,CaseInsensitive(String) (Utilities|Struct|BreakEventRouterEventString _payload) (Variables|Default|GetEventEndRouterEvent))
                    (Variables|Default|SetEventEnded true)
                    (CallFunction|MoveZone)))))))))))

(event Custom|PC_PostEvent
  (Variables|Default|SetEventStarted true)
  (Variables|Default|SetEventEnded true)
  (CallFunction|MoveZone))

(event Custom|DisableLeaderBoard
  (Variables|Default|SetEventStarted true)
  (bind _outfortplayerpawns (Fortnite|GetFortPlayerPawns))
  (for _array_element _outfortplayerpawns
    (bind _asplayer_pawn_athena (Utilities|Casting|CastToPlayerPawn_Athena _array_element))
    (Development|PrintString (Utilities|String|Append "RemovedLeaderBoard
" (Utilities|GetDisplayName _asplayer_pawn_athena)))
    (CallFunction|RemoveLeaderboardInput _asplayer_pawn_athena)))

(event Custom|CheatOverlap
  (params (OverlapPawn :type Player%20Pawn%20Athena%20Object%20Reference))
  (bind _returnvalue (Ability|GetFortAbilitySystemComponent (Variables|Default|GetOverlapPawn)))
  (bind _returnvalue_2 (Ability|GetFortAbilitySystemComponent (Variables|Default|GetOverlapPawn)))
  (bind _overlappawn (Variables|Default|GetOverlapPawn))
  (Variables|Default|SetOverlapPawn OverlapPawn)
  (Utilities|IsValid _overlappawn
    (:Is%20Valid
      (if (Utilities|IsValid (Fortnite|GetControllerfromActor _overlappawn) :!class /Script/Engine.KismetSystemLibrary)
        (if (not (Pawn|IsBotControlled _overlappawn))
          (CustomVerb|BroadcastCustomVerbMessage (CustomVerb|MakeCustomVerbMessage _overlappawn "(TagName=\"CustomVerb.PistolCanary.ShowUrgentQuest\")"))
          (CallFunction|ApplyInventoryEffect _overlappawn)
          (if (not (Variables|Default|GetEventStarted))
            (GameplayEffects|ApplyGameplayEffectToTarget :self _returnvalue :GameplayEffectClass (Variables|Default|GetKillCountPassive) :Target _returnvalue :Level 0.0)
            (GameplayEffects|ApplyGameplayEffectToTarget :self _returnvalue :GameplayEffectClass (Variables|Default|GetGEDamageImmune) :Target _returnvalue :Level 0.0)
            (GameplayEffects|ApplyGameplayEffectToTarget :self _returnvalue :GameplayEffectClass (Variables|Default|GetGEDamageRules) :Target _returnvalue :Level 0.0)
            (CallFunction|RefreshPlayersinCombat _overlappawn)
            (else
              _))
          (if (not (Variables|Default|GetEventEnded))
            (bind _returnvalue_3 (Actor|GetActorOfClass "/PistolCanary_Training/Gameplay/BP_PistolCanary_EventZone_PreEvent_GameplayManager.BP_PistolCanary_EventZone_PreEvent_GameplayManager_C"))
            (Utilities|IsValid _returnvalue_3
              (:Is%20Valid
                (Class|BPPistolCanaryEventZonePreEventGameplayManager|AddPlayertoLeaderBoard _returnvalue_3 _overlappawn)
                (CallFunction|AddPlayertoLeaderboard _overlappawn)
                (if (not (Variables|Default|GetEventStarted))
                  _
                  (else
                    _))))
            (else
              _))
          (GameplayEffects|ApplyGameplayEffectToTarget :self _returnvalue_2 :GameplayEffectClass "/PistolCanary_Training/Gameplay/Abilities/GE_PistolCanary_EventDamageImmune_EventRules.GE_PistolCanary_EventDamageImmune_EventRules_C" :Target _returnvalue_2 :Level 0.0)
          (if (not (Variables|Default|GetEventEnded))
            (GameplayEffects|ApplyGameplayEffectToTarget :self _returnvalue_2 :GameplayEffectClass "/PistolCanary_Training/Gameplay/Effects/GE_PistolCanary_EventNoAmmoCost.GE_PistolCanary_EventNoAmmoCost_C" :Target _returnvalue_2 :Level 0.0))
          (else
            (if (not (Variables|Default|GetEventStarted))
              (if (not (GameplayTags|HasAnyMatchingGameplayTags _overlappawn (Variables|Default|GetEvent%20NPC%20Tag%20Container)))
                (Utilities|FlowControl|Delay 0.1)
                (Environment|ApplyGlobalEnvironmentDamagetoActor _overlappawn 5e+02)))))))))

(event OnComponentBeginOverlap(Overlap_Volume)
  (params (OverlappedComponent :type Primitive%20Component%20Object%20Reference) (OtherActor :type Actor%20Object%20Reference) (OtherComp :type Primitive%20Component%20Object%20Reference) (OtherBodyIndex :type Integer) (bFromSweep :type Boolean) (SweepResult :type Hit%20Result%20Structure%20%28by%20ref%29))
  (bind _returnvalue (Ability|GetFortAbilitySystemComponent (Variables|Default|GetOverlapPawn)))
  (bind _returnvalue_2 (Ability|GetFortAbilitySystemComponent (Variables|Default|GetOverlapPawn)))
  (bind _overlappawn (Variables|Default|GetOverlapPawn))
  (Networking|SwitchHasAuthority
    (:Authority
      (bind _asplayer_pawn_athena (Utilities|Casting|CastToPlayerPawn_Athena OtherActor)
        (:then
          (Variables|Default|SetOverlapPawn _asplayer_pawn_athena)
          (Utilities|IsValid _overlappawn
            (:Is%20Valid
              (if (Utilities|IsValid (Fortnite|GetControllerfromActor _overlappawn) :!class /Script/Engine.KismetSystemLibrary)
                (if (not (Pawn|IsBotControlled _overlappawn))
                  (CustomVerb|BroadcastCustomVerbMessage (CustomVerb|MakeCustomVerbMessage _overlappawn "(TagName=\"CustomVerb.PistolCanary.ShowUrgentQuest\")"))
                  (CallFunction|ApplyInventoryEffect _overlappawn)
                  (if (not (Variables|Default|GetEventStarted))
                    (GameplayEffects|ApplyGameplayEffectToTarget :self _returnvalue :GameplayEffectClass (Variables|Default|GetKillCountPassive) :Target _returnvalue :Level 0.0)
                    (GameplayEffects|ApplyGameplayEffectToTarget :self _returnvalue :GameplayEffectClass (Variables|Default|GetGEDamageImmune) :Target _returnvalue :Level 0.0)
                    (GameplayEffects|ApplyGameplayEffectToTarget :self _returnvalue :GameplayEffectClass (Variables|Default|GetGEDamageRules) :Target _returnvalue :Level 0.0)
                    (CallFunction|RefreshPlayersinCombat _overlappawn)
                    (else
                      _))
                  (if (not (Variables|Default|GetEventEnded))
                    (bind _returnvalue_3 (Actor|GetActorOfClass "/PistolCanary_Training/Gameplay/BP_PistolCanary_EventZone_PreEvent_GameplayManager.BP_PistolCanary_EventZone_PreEvent_GameplayManager_C"))
                    (Utilities|IsValid _returnvalue_3
                      (:Is%20Valid
                        (Class|BPPistolCanaryEventZonePreEventGameplayManager|AddPlayertoLeaderBoard _returnvalue_3 _overlappawn)
                        (CallFunction|AddPlayertoLeaderboard _overlappawn)
                        (if (not (Variables|Default|GetEventStarted))
                          _
                          (else
                            _))))
                    (else
                      _))
                  (GameplayEffects|ApplyGameplayEffectToTarget :self _returnvalue_2 :GameplayEffectClass "/PistolCanary_Training/Gameplay/Abilities/GE_PistolCanary_EventDamageImmune_EventRules.GE_PistolCanary_EventDamageImmune_EventRules_C" :Target _returnvalue_2 :Level 0.0)
                  (if (not (Variables|Default|GetEventEnded))
                    (GameplayEffects|ApplyGameplayEffectToTarget :self _returnvalue_2 :GameplayEffectClass "/PistolCanary_Training/Gameplay/Effects/GE_PistolCanary_EventNoAmmoCost.GE_PistolCanary_EventNoAmmoCost_C" :Target _returnvalue_2 :Level 0.0))
                  (else
                    (if (not (Variables|Default|GetEventStarted))
                      (if (not (GameplayTags|HasAnyMatchingGameplayTags _overlappawn (Variables|Default|GetEvent%20NPC%20Tag%20Container)))
                        (Utilities|FlowControl|Delay 0.1)
                        (Environment|ApplyGlobalEnvironmentDamagetoActor _overlappawn 5e+02)))))))))
        (:CastFailed
          (bind _asfort_athena_vehicle (Utilities|Casting|CastToFortAthenaVehicle OtherActor))
          (Vehicle|WakeUp _asfort_athena_vehicle)
          (Components|Movement|StopMovementImmediately (Pawn|GetMovementComponent _asfort_athena_vehicle))
          (Ability|GameplayEffect|AddLooseGameplayTags OtherActor (Variables|Default|GetVehicleDisableTags) true)
          (Utilities|Array|Add (Variables|Default|GetDisabledVehicles) _asfort_athena_vehicle))))
    (:Remote
      (if (not (Variables|Default|GetEventStarted))
        (bind _asplayer_pawn_athena_2 (Utilities|Casting|CastToPlayerPawn_Athena OtherActor))
        (CallFunction|AddLeaderboardInput _asplayer_pawn_athena_2)))))

(event OnComponentEndOverlap(Overlap_Volume)
  (params (OverlappedComponent :type Primitive%20Component%20Object%20Reference) (OtherActor :type Actor%20Object%20Reference) (OtherComp :type Primitive%20Component%20Object%20Reference) (OtherBodyIndex :type Integer))
  (Networking|SwitchHasAuthority
    (:Authority
      (if (not (Variables|Default|GetEventStarted))
        _
        (elif (Variables|Default|GetEventEnded)
          _))
      (bind _asplayer_pawn_athena (Utilities|Casting|CastToPlayerPawn_Athena OtherActor)
        (:then
          (if (not (Pawn|IsBotControlled _asplayer_pawn_athena))
            (Utilities|IsValid _asplayer_pawn_athena
              (:Is%20Valid
                (CallFunction|RemovePlayerfromPvP _asplayer_pawn_athena)
                (GameplayEffects|RemoveActiveEffectswithSourceTags (Ability|GetFortAbilitySystemComponent _asplayer_pawn_athena) "(GameplayTags=((TagName=\"Gameplay.Mod.FireRate\")))")
                ; Give enough time for first aura to turn off
                (Utilities|FlowControl|Delay 0.1)
                (GameplayEffects|RemoveActiveGameplayEffectbySourceEffect (Ability|GetFortAbilitySystemComponent _asplayer_pawn_athena) (Variables|Default|GetEventZoneAuraEffect) (Ability|GetFortAbilitySystemComponent _asplayer_pawn_athena) -1)
                (GameplayEffects|RemoveActiveGameplayEffectbySourceEffect (Ability|GetFortAbilitySystemComponent _asplayer_pawn_athena) (Variables|Default|GetKillCountPassive) (Ability|GetFortAbilitySystemComponent _asplayer_pawn_athena) -1)
                (GameplayEffects|RemoveActiveGameplayEffectbySourceEffect (Ability|GetFortAbilitySystemComponent _asplayer_pawn_athena) (Variables|Default|GetGEDamageImmune) (Ability|GetFortAbilitySystemComponent _asplayer_pawn_athena) -1)
                (GameplayEffects|RemoveActiveGameplayEffectbySourceEffect (Ability|GetFortAbilitySystemComponent _asplayer_pawn_athena) (Variables|Default|GetGEDamageRules) (Ability|GetFortAbilitySystemComponent _asplayer_pawn_athena) -1)
                (bind _returnvalue (Actor|GetActorOfClass "/PistolCanary_Training/Gameplay/BP_PistolCanary_EventZone_PreEvent_GameplayManager.BP_PistolCanary_EventZone_PreEvent_GameplayManager_C"))
                (Class|BPPistolCanaryEventZonePreEventGameplayManager|RemovePlayerfromLeaderBoard _returnvalue _asplayer_pawn_athena)
                (CallFunction|RemovePlayerfromLeaderboard _asplayer_pawn_athena)
                (if (Variables|Default|GetEventEnded)
                  (GameplayEffects|RemoveActiveGameplayEffectbySourceEffect (Ability|GetFortAbilitySystemComponent _asplayer_pawn_athena) "/PistolCanary_Training/Gameplay/Abilities/GE_PistolCanary_EventDamageImmune_EventRules.GE_PistolCanary_EventDamageImmune_EventRules_C" (Ability|GetFortAbilitySystemComponent _asplayer_pawn_athena) -1))))))
        (:CastFailed
          (Utilities|Casting|CastToFortAthenaVehicle OtherActor)
          (Ability|GameplayEffect|RemoveLooseGameplayTags OtherActor (Variables|Default|GetVehicleDisableTags)))))
    (:Remote
      (bind _asplayer_pawn_athena_2 (Utilities|Casting|CastToPlayerPawn_Athena OtherActor))
      (CallFunction|RemoveLeaderboardInput _asplayer_pawn_athena_2))))

(event Custom|MoveZone
  (bind (_returnvalue_location _returnvalue_rotation) (Math|Transform|BreakTransform (Transformation|GetWorldTransform (Variables|Default|GetBarrier_Mesh))))
  (bind _returnvalue (Transformation|GetWorldScale (Variables|Default|GetBarrier_Mesh)))
  (bind _barrier_mesh (Variables|Default|GetBarrier_Mesh))
  (bind _posteventscale (Variables|Default|GetPostEventScale))
  ; Move Zone at end of event
  ;;[0.01777 0.15 0.08989 0.5] Turn off the Zone Aura directly.
  ;; Add a hotfixable check in case we find the old way is better
  ;; Change before CERT
  (if (ScalableFloat|GetValueasBool (Variables|Default|GetNoShrinkingZone))
    (Rendering|SetActorHiddenInGame self true)
    (else
      (Networking|SwitchHasAuthority
        (:Authority
          (Components|Timeline|SetPlayRate (Variables|BP_PistolCanary_NoFightBarrier|GetZoneMovement) 0.01)
          (bind (_direction _movementalpha) (Timeline|ZoneMovement PlayFromStart)
            (:Update
              (Transformation|SetWorldTransform _barrier_mesh (Math|Transform|MakeTransform (Math|Vector|Lerp(Vector) (Transformation|GetWorldLocation _barrier_mesh) (Variables|Default|GetPostEventLocation) _movementalpha) (Transformation|GetWorldRotation _barrier_mesh) (Math|Vector|Lerp(Vector) _returnvalue _posteventscale _movementalpha)) false false)
              (Variables|Default|SetUpdateScale (Math|Vector|Lerp(Vector) _returnvalue _posteventscale _movementalpha))
              (if (Math|Vector|VectorIsNearlyZero (- (Transformation|GetWorldLocation (Variables|Default|GetBarrier_Mesh)) (Variables|Default|GetPostEventLocation)) 2e+02)
                (Timeline|ZoneMovement Stop)))
            (:Finished
              (Transformation|SetWorldTransform (Variables|Default|GetBarrier_Mesh) (Math|Transform|MakeTransform (Variables|Default|GetPostEventLocation) _returnvalue_rotation (Variables|Default|GetPostEventScale)) false false))))))))

(fn AddLeaderboardInput
  (params (PlayerPawn :type Player%20Pawn%20Athena%20Object%20Reference))
  (bind _returnvalue (Fortnite|GetFortPlayerControllerfromActor PlayerPawn))
  (bind _returnvalue_2 (Actor|GetComponentbyClass _returnvalue "/PistolCanary_Training/Input/BP_PistolCanary_Leaderboard_InputComponent.BP_PistolCanary_Leaderboard_InputComponent_C"))
  (Networking|SwitchHasAuthority
    (:Remote
      (Utilities|IsValid PlayerPawn
        (:Is%20Valid
          (Utilities|IsValid _returnvalue
            (:Is%20Valid
              (Utilities|IsValid _returnvalue_2
                (:Is%20Valid
                  (Class|SpecialEventLeaderboardInputComponent|SetupInput _returnvalue_2))
                (:Is%20Not%20Valid
                  (bind _returnvalue_3 (Game|AddComponentbyClass _returnvalue "/PistolCanary_Training/Input/BP_PistolCanary_Leaderboard_InputComponent.BP_PistolCanary_Leaderboard_InputComponent_C"))
                  (Class|SpecialEventLeaderboardInputComponent|SetupInput _returnvalue_3))))))))))

(fn AddPlayerToLeaderboard
  (params (PlayerPawn :type Player%20Pawn%20Athena%20Object%20Reference))
  (bind _returnvalue (Class|SpecialEventMeshLeaderboardManager|GetMeshLeaderboardManager true))
  (bind _playerstate (Variables|/Script/Engine.Pawn|GetPlayerState PlayerPawn))
  (Networking|SwitchHasAuthority
    (:Authority
      (Utilities|IsValid PlayerPawn
        (:Is%20Valid
          (Utilities|IsValid _playerstate
            (:Is%20Valid
              (Utilities|IsValid _returnvalue
                (:Is%20Valid
                  (Class|SpecialEventMeshLeaderboardManager|AddPlayertoLeaderboard _returnvalue _playerstate))))))))))

(fn ApplyInventoryEffect
  (params (PlayerPawnAthena :type Fort%20Player%20Pawn%20Athena%20Object%20Reference))
  (if (not (Variables|Default|GetEventEnded))
    (GameplayEffects|ApplyGameplayEffectToTarget :self (Ability|GetFortAbilitySystemComponent PlayerPawnAthena) :GameplayEffectClass (Variables|Default|GetEventZoneAuraEffect) :Target (Ability|GetFortAbilitySystemComponent PlayerPawnAthena) :Level 0.0)))

(fn CheckLoadedWithCheats
  (bind _outactors (Actor|GetAllActorsOfClass "/PistolCanary_Training/Gameplay/BP_PistolCanary_EventZoneManager.BP_PistolCanary_EventZoneManager_C"))
  (for _array_element _outactors
    (Variables|Default|SetCheatActivated (Variables|/PistolCanary_Training/Gameplay/BP_PistolCanary_EventZoneManager.BP_PistolCanary_EventZoneManager_C|GetLoadedWithCheats _array_element)))
  (Development|PrintString (Utilities|String|Append "I was loaded via cheat: " (Utilities|String|ToString(Boolean) (Variables|Default|GetCheatActivated))))
  (if (Variables|Default|GetCheatActivated)
    (bind _outactors_2 (Fortnite|GetAllActorsOfClassinRange "/Script/FortniteGame.FortPlayerPawn" (Transformation|GetActorLocation self) 1.874e+04))
    (for _array_element_2 _outactors_2
      (bind _asplayer_pawn_athena (Utilities|Casting|CastToPlayerPawn_Athena _array_element_2))
      (CallFunction|CheatOverlap _asplayer_pawn_athena))))

(fn OnRep_RemoveLeaderBoard
  (if (not (Networking|IsServer))
    (CallFunction|DisableLeaderBoard)))

(fn OnRep_ShowZone
  (bind _showzone (Variables|Default|GetShowZone))
  (Variables|/Script/Engine.PrimitiveComponent|SetbGenerateOverlapEvents _showzone (Variables|Default|GetOverlap_Volume))
  (Networking|CanExecuteCosmeticEvents
    (:True
      (if _showzone
        (Rendering|SetActorHiddenInGame false)
        (Development|PrintString "I should be visible" true true "(R=0.000000,G=0.660000,B=1.000000,A=1.000000)" 2e+01)
        (else
          (Rendering|SetActorHiddenInGame true))))
    (:False
      (Collision|SetCollisionEnabled (Variables|Default|GetOverlap_Volume) (select _showzone "QueryOnly" "NoCollision")))))

(fn OnRep_UpdateScale
  (Networking|CanExecuteCosmeticEvents
    (:True
      (Transformation|SetActorScale3D self (Variables|Default|GetUpdateScale)))))

(fn RemoveLeaderboardInput
  (params (PlayerPawn :type Player%20Pawn%20Athena%20Object%20Reference))
  (bind _returnvalue (Fortnite|GetFortPlayerControllerfromActor PlayerPawn))
  (bind _returnvalue_2 (Actor|GetComponentbyClass _returnvalue "/PistolCanary_Training/Input/BP_PistolCanary_Leaderboard_InputComponent.BP_PistolCanary_Leaderboard_InputComponent_C"))
  (Networking|SwitchHasAuthority
    (:Remote
      (Utilities|IsValid PlayerPawn
        (:Is%20Valid
          (Utilities|IsValid _returnvalue
            (:Is%20Valid
              (Utilities|IsValid _returnvalue_2
                (:Is%20Valid
                  (Class|SpecialEventLeaderboardInputComponent|RemoveInputMapping _returnvalue_2))))))))))

(fn RemovePlayerFromLeaderboard
  (params (PlayerPawn :type Player%20Pawn%20Athena%20Object%20Reference))
  (bind _returnvalue (Class|SpecialEventMeshLeaderboardManager|GetMeshLeaderboardManager true))
  (bind _playerstate (Variables|/Script/Engine.Pawn|GetPlayerState PlayerPawn))
  (Networking|SwitchHasAuthority
    (:Authority
      (Utilities|IsValid PlayerPawn
        (:Is%20Valid
          (Utilities|IsValid _playerstate
            (:Is%20Valid
              (Utilities|IsValid _returnvalue
                (:Is%20Valid
                  (Class|SpecialEventMeshLeaderboardManager|RemovePlayerfromLeaderboard _returnvalue _playerstate))))))))))

(fn ConstructionScript)
